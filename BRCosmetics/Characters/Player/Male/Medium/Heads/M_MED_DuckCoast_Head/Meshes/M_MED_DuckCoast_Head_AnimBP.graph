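# Animation Blueprint for the male DuckCoast character head that mirrors facial eye animation into a material overlay.
# When character customization loads, it gets or creates a dynamic overlay material on the part's skeletal mesh and caches it as OverlayMat.
# Each animation update passes through a zero-length Delay, then copies left and right blink, upper-lid-down and inner-squint curve values into same-named scalar parameters on that material.
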AnimBlueprint /BRCosmetics/Characters/Player/Male/Medium/Heads/M_MED_DuckCoast_Head/Meshes/M_MED_DuckCoast_Head_AnimBP
target_skeleton=/Game/Characters/Player/Common/Fortnite_Base_Head/Fortnite_Base_Head_Export_Skeleton
graphs [EventGraph {nodes [
   Event BlueprintUpdateAnimation()
   > Delay(Duration=0)
   > SetScalarParameterValue(
    self=Get OverlayMat() -> (OverlayMat:Material Instance Dynamic Object Reference)
    ParameterName=L_blink_pose
    Value=GetCurveValue(CurveName=L_blink_pose)
   )
   > SetScalarParameterValue(self=Get OverlayMat(), ParameterName=R_blink_pose, Value=GetCurveValue(CurveName=R_blink_pose))
   > SetScalarParameterValue(self=Get OverlayMat(), ParameterName=L_upperLidDown_pose, Value=GetCurveValue(CurveName=L_upperLidDown_pose))
   > SetScalarParameterValue(self=Get OverlayMat(), ParameterName=R_upperLidDown_pose, Value=GetCurveValue(CurveName=R_upperLidDown_pose))
   > SetScalarParameterValue(self=Get OverlayMat(), ParameterName=L_squint_inner_pose, Value=GetCurveValue(CurveName=L_squint_inner_pose))
   > SetScalarParameterValue(self=Get OverlayMat(), ParameterName=R_squint_inner_pose, Value=GetCurveValue(CurveName=R_squint_inner_pose))
   Event OnCharacterCustomizationLoaded()
   > n1 = GetOrCreateDynamicOverlayMaterial(MeshComponent=GetPartSkeletalMeshComponent(), MaterialName=None)
   > Set OverlayMat(OverlayMat:Material Instance Dynamic Object Reference=@n1) -> (Output_Get:Material Instance Dynamic Object Reference)
  ]}]
variables [OverlayMat {type=MaterialInstanceDynamic}]
